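# Niagara system for the lightning effect of the FumeFleeceClap player emote, timed through staggered emitter delays.
# Mesh lightning arcs, flash sprites and expanding rings attach to sampled skeletal mesh bones or sockets, with animated dynamic material parameters shaping the look.
# ParticleDecal spawns from Lightning particles through an attribute reader.
# User.LightningColor tints the effect, with main bolts switching to a multicolor palette mid-emote; some brightness scales for front end and time of day.
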NiagaraSystem /BRCosmetics/Animation/Game/MainPlayer/Emotes/FumeFleeceClap/FX/NS_Emote_FumeFleeceClap_Lightning
dependencies {
 usedDataInterfaces [
  &NiagaraDataInterfaceCurve
  &NiagaraDataInterfaceMeshRendererInfo
  &NiagaraDataInterfaceArrayFloat3
  &NiagaraDataInterfaceVectorCurve
  &NiagaraDataInterfaceArrayColor
  &NiagaraDataInterfaceParticleRead
  &NiagaraDataInterfaceColorCurve
  &NiagaraDataInterfaceSkeletalMesh
 ]
 usedDynamicInputs [&MakeCustomVectorFromBool, &SetBoolByFloatComparison]
 usedModules [
  &SystemState
  &self:SystemSpawnScript.SystemScriptSource.SystemScriptGraph.NiagaraNodeAssignment_0.SetVariables_#1
 ]
}
emitters {
 CircularLightning {
  emitterUpdateScript [
   &EmitterState(Life Cycle Mode=::NewEnumerator1, Loop Behavior=::NewEnumerator1, Loop Delay=0.5, Loop Duration=1.4, UseLoopDelay=true)
   &SpawnRate(
    SpawnRate=&FloatFromCurve(
     FloatCurve=DataInterface() {propertyValues {Curve {keys [
         {interpMode=RCIM_Cubic, tangentMode=RCTM_User, time=0, value=0.22}
         {arriveTangent=4.4807, interpMode=RCIM_Cubic, leaveTangent=4.4807, tangentMode=RCTM_User, time=0.7269, value=1.5805}
        ]}}}
     CurveIndex=$Emitter.NormalizedLoopAge
     Scale Curve=15
    )
    Use Spawn Probability=false
   )
  ]
  particleSpawnScript [
   &InitializeParticle(
    Color Maximum=<rgba 2.2264 5 4.6542 1>
    Color Minimum=<rgba 5 4.2427 2.0862 1>
    Color Mode=::NewEnumerator2
    Lifetime Max=0.07
    Lifetime Min=0.2
    Lifetime Mode=::NewEnumerator1
    Mesh Scale=&Multiply_VectorByFloat(
     Vector=&RandomRangeVector(Minimum=<0.5 0.9 0.9>, Maximum=<0.5 0.7311 0.6712>)
     Float=&RandomRangeFloat(Minimum=1.6, Maximum=1.6)
    )
    Mesh Scale Mode=::NewEnumerator4
    Position Offset=<0 0 20>
    Sprite Size Mode=::NewEnumerator3
    UsePositionOffset=false
    Write Material Random=true
   )
   &SkeletalMeshLocation(Bone / Socket Index=1, Bone Sampling Mode=::NewEnumerator4, MaskSpawnGroup=false, Skeletal Mesh=$System.Skeletal Mesh)
   &CameraOffset(Camera Offset Amount=&RandomRangeFloat(Minimum=20, Maximum=20))
   &ShapeLocation(
    Handle Radius=0
    Large Radius=20
    MaskSpawnGroup=false
    Offset=<0 10 0>
    Offset Mode=::NewEnumerator0
    Override Seed=false
    Rotation Mode=::NewEnumerator2
    Shape Origin=$Particles.SkeletalMeshLocation.SampledPosition
    Shape Primitive=::NewEnumerator3
    Surface Distribution=1
    UseOwnerScale=true
    V Distribution=1
    Yaw / Pitch / Roll=<0 0 0.25>
   )
   &AddVelocity(
    Module.Use Falloff Distance=false
    Origin Offset=<0 10 0>
    Origin Offset Coordinate Space=::Simulation
    Override Seed=false
    Velocity Mode=::NewEnumerator1
    Velocity Origin=$Particles.SkeletalMeshLocation.SampledPosition
    Velocity Speed=0.1
   )
   &PartitionParticles(Number Of Partitions=15, Particles Per Partition=1, Partition Mode:ENiagara_PartitionMode_v2=::NewEnumerator1)
   &InitialMeshOrientation(Mesh Orientation Mode=::NewEnumerator2, Orientation Vector=$StackContext.Velocity, Rotation=<0.2519 0 0>, Use Rotation Vector=true) {module=InitialMeshOrientation001}
  ]
  particleUpdateScript [
   &ParticleState()
   &DynamicMaterialParameters(
    Index 0 Param 1=&MakeFloatFromVector(Vector=$Particles.Scale, Channel=::NewEnumerator0)
    Index 0 Param 2=&RandomRangeFloat(Minimum=3, Maximum=10)
    Index 0 Param 3=&FloatFromCurve(
     FloatCurve=DataInterface() {propertyValues {Curve {keys [{interpMode=RCIM_Cubic, leaveTangent=8.1085, tangentMode=RCTM_Break, time=0.8568, value=1}, {tangentMode=RCTM_User, time=1, value=0}]}}}
     Scale Curve=0.005
    )
    Index 0 Param 4=&RandomRangeFloat(Minimum=0.4, Maximum=0.2)
    Index 1 Param 1=&FloatFromCurve(FloatCurve=DataInterface() {propertyValues {Curve {keys [
         {interpMode=RCIM_Cubic, leaveTangent=1.1034, tangentMode=RCTM_Break, time=0, value=0}
         {arriveTangent=2.2399, tangentMode=RCTM_User, time=0.5, value=1}
        ]}}})
    Index 1 Param 2=&FloatFromCurve(FloatCurve=DataInterface() {propertyValues {Curve {keys [<at 0.5 0>, {arriveTangent=3.2053, tangentMode=RCTM_User, time=0.9, value=1}]}}})
    Index 1 Param 3=&FloatFromCurve(FloatCurve=DataInterface() {propertyValues {Curve {keys [
         <at 0 0 RCIM_Cubic>
         {arriveTangent=0.5046, interpMode=RCIM_Cubic, leaveTangent=0.5046, time=0.8846, value=0.0963}
         {arriveTangent=2.773, interpMode=RCIM_Cubic, leaveTangent=2.773, tangentMode=RCTM_User, time=1, value=0.5046}
        ]}}})
    Index 1 Param 4=&FloatFromCurve(
     FloatCurve=DataInterface() {propertyValues {Curve {keys [
         {interpMode=RCIM_Cubic, leaveTangent=0.1169, tangentMode=RCTM_Break, time=0, value=0}
         {arriveTangent=2.5708, interpMode=RCIM_Cubic, leaveTangent=2.5708, tangentMode=RCTM_User, time=1, value=1}
        ]}}}
     Scale Curve=&RandomRangeFloat(Minimum=1, Maximum=40)
    )
    Index 2 Param 1=&FloatFromCurve(FloatCurve=DataInterface() {propertyValues {Curve {keys [
         {interpMode=RCIM_Cubic, leaveTangent=-0.0025, tangentMode=RCTM_Break, time=0, value=0.5}
         {arriveTangent=-0.0316, interpMode=RCIM_Cubic, tangentMode=RCTM_Break, time=1, value=-0.2}
        ]}}})
    Index 2 Param 2=&FloatFromCurve(
     FloatCurve=DataInterface() {propertyValues {Curve {keys [
         {interpMode=RCIM_Cubic, tangentMode=RCTM_User, time=0, value=0}
         {arriveTangent=2.0522, interpMode=RCIM_Cubic, leaveTangent=2.0522, tangentMode=RCTM_User, time=1, value=1}
        ]}}}
     Scale Curve=&RandomRangeFloat(Minimum=0.4, Maximum=1)
    )
    Module.Index0Float0Write=true
    Module.Index0Float1Write=true
    Module.Index0Float2Write=true
    Module.Index0Float3Write=true
    Module.Index1Float0Write=true
    Module.Index1Float1Write=true
    Module.Index1Float2Write=true
    Module.Index1Float3Write=true
    Module.Index2Float0Write=true
    Module.Index2Float1Write=true
    Module.Index2Float2Write=false
    Module.Index2Float3Write=false
    Module.Index3Float0Write=false
    Module.Index3Float1Write=false
    Module.Index3Float2Write=false
    Module.Index3Float3Write=true
    Write Parameter Index 1=::NewEnumerator1
    Write Parameter Index 2=::NewEnumerator1
    Write Parameter Index 3=::NewEnumerator1
   )
   /SetVariables(Particles.Glow Color=&MakeLinearColorFromVectorAndFloat(
     "Vector (RGB)"=&Lerp_Vector(
      A=&MakeVectorFromLinearColorRGB(Color=&Multiply_LinearColorByFloat(Linear Color=$User.LightningColor, Float=1.5))
      B=&MakeVectorFromLinearColorRGB(Color=&SelectLinearColorFromArray(
        Array Sampling Mode=::NewEnumerator1
        Direct Array Index=$Engine.ExecIndex
        Direct Array Mode=::NewEnumerator1
        Color Selection Array=DataInterface() {propertyValues {ColorData [<rgba 5 4.6482 0.6567 1>, <rgba 0.0274 5 1.1324 1>, <rgba 1.0025 2.8799 5 1>, <rgba 0.5806 0.5514 5 1>]}}
        Random Seed=?
       ))
      Alpha=&FloatFromCurve(
       FloatCurve=DataInterface() {propertyValues {Curve {
          defaultValue=1
          keys [{interpMode=RCIM_Cubic, tangentMode=RCTM_User, time=1.5251, value=0}, {interpMode=RCIM_Cubic, tangentMode=RCTM_User, time=1.545, value=1}]
         }}}
       CurveIndex=$Emitter.Age
      )
     )
     "Float (Alpha)"=&Multiply_Float(
      A=&FloatFromCurve(FloatCurve=DataInterface() {propertyValues {Curve {keys [
           {interpMode=RCIM_Cubic, tangentMode=RCTM_User, time=0, value=1}
           <at 0.7 1>
           {arriveTangent=-3.6307, interpMode=RCIM_Cubic, leaveTangent=-3.6307, tangentMode=RCTM_User, time=0.9, value=0}
          ]}}})
      B=4
     )
    )) {isSetParameters=true}
   /SetVariables(Particles.Glow Dy02:Vector4f=&MakeVector4(
     X=&FloatFromCurve(FloatCurve=DataInterface() {propertyValues {Curve {keys [<at 0 0>,<at 0.7 1>]}}})
     Y=0.7
     Z=0
     W=0
    )) {isSetParameters=true}
   &SolveForcesAndVelocity(Clamp Velocity=false, Limit Acceleration=false)
   &ScaleColor(
    Scale Alpha=&Multiply_Float(
     A=20
     B=&FloatFromCurve(FloatCurve=DataInterface() {propertyValues {Curve {keys [
          {tangentMode=RCTM_User, time=0, value=1}
          {arriveTangent=-0.0405, interpMode=RCIM_Cubic, leaveTangent=-0.0405, tangentMode=RCTM_User, time=0.7009, value=1}
          {arriveTangent=-8.0089, interpMode=RCIM_Cubic, leaveTangent=-8.0089, tangentMode=RCTM_User, time=1, value=0}
         ]}}})
    )
    Scale RGB=&Lerp_Vector(
     A=&MakeVectorFromLinearColorRGB(Color=&Multiply_LinearColorByFloat(Linear Color=$User.LightningColor, Float=1.5))
     B=&MakeVectorFromLinearColorRGB(Color=&SelectLinearColorFromArray(
       Array Sampling Mode=::NewEnumerator1
       Direct Array Index=$Engine.ExecIndex
       Direct Array Mode=::NewEnumerator1
       Color Selection Array=DataInterface() {propertyValues {ColorData [<rgba 5 4.6482 0.6567 1>, <rgba 0.0274 5 1.1324 1>, <rgba 1.0025 2.8799 5 1>, <rgba 0.5806 0.5514 5 1>]}}
       Random Seed=?
      ))
     Alpha=&FloatFromCurve(
      FloatCurve=DataInterface() {propertyValues {Curve {
         defaultValue=1
         keys [{interpMode=RCIM_Cubic, tangentMode=RCTM_User, time=1.5251, value=0}, {interpMode=RCIM_Cubic, tangentMode=RCTM_User, time=1.545, value=1}]
        }}}
      CurveIndex=$Emitter.Age
     )
    )
    ScaleA=true
    ScaleRGB=true
   )
  ]
  rendererClasses [&NiagaraMeshRendererProperties]
  renderers [
   {rendererClass=&NiagaraMeshRendererProperties, rendererIndex=0}
   {rendererClass=&NiagaraMeshRendererProperties, rendererIndex=1}
  ]
 }
 Flash {
  emitterUpdateScript [
   &EmitterState(Life Cycle Mode=::NewEnumerator1, Loop Behavior=::NewEnumerator1, Loop Delay=1.55, Loop Duration=3, UseLoopDelay=true)
   &SpawnBurst_Instantaneous(Spawn Time=0.03, Use Spawn Probability=false, UseLoopCountLimit=false)
   &SpawnBurst_Instantaneous(Spawn Group=1, Use Spawn Probability=false, UseLoopCountLimit=false) {module=SpawnBurst_Instantaneous001}
  ]
  particleSpawnScript [
   /SetVariables(Particles.VisibilityTag=&SpawnGroup()) {isSetParameters=true}
   &InitializeParticle(
    Color=&Multiply_LinearColorByFloat(
     Linear Color=&MaskLinearColorBySpawnGroup(Masked Linear Color=<rgba 1 0.8953 0.6256 0.6>, Passthrough Linear Color=<rgba 1.5 1.2166 0.487 1>)
     Float=&MaskFloatBySpawnGroup(Masked Float=1, Passthrough Float=1.5)
    )
    Lifetime=0.605
    Sprite Size Mode=::NewEnumerator3
    Uniform Sprite Size=&MaskFloatBySpawnGroup(Masked Float=115, Passthrough Float=150)
    UsePositionOffset=false
    Write Material Random=true
   )
   &CameraOffset(Camera Offset Amount=&MaskFloatBySpawnGroup(Masked Float=20, Passthrough Float=-20))
   &SkeletalMeshLocation(Bone / Socket Index=1, Bone Sampling Mode=::NewEnumerator4, MaskSpawnGroup=false, Skeletal Mesh=$System.Skeletal Mesh)
  ]
  particleUpdateScript [
   &SolveForcesAndVelocity(Clamp Velocity=false, Limit Acceleration=false)
   &ParticleState()
   &SpriteRotationRate(Rotation Rate=&FloatFromCurve(
     FloatCurve=DataInterface() {propertyValues {Curve {
        defaultValue=1
        keys [
         {interpMode=RCIM_Cubic, tangentMode=RCTM_User, time=0, value=0}
         {arriveTangent=7.1184, interpMode=RCIM_Cubic, leaveTangent=7.1184, tangentMode=RCTM_User, time=0.467, value=1}
        ]
       }}}
     Scale Curve=400
    ))
   &ScaleSpriteSize(Uniform Curve Sprite Scale=DataInterface() {propertyValues {Curve {
       defaultValue=0
       keys [
        {arriveTangent=2.1946, interpMode=RCIM_Cubic, leaveTangent=2.1946, tangentMode=RCTM_User, time=0, value=0.3921}
        {interpMode=RCIM_Cubic, tangentMode=RCTM_User, time=0.0997, value=1.8807}
        {arriveTangent=-0.9926, interpMode=RCIM_Cubic, leaveTangent=-0.9926, time=0.2, value=1.3}
        {arriveTangent=-1.625, interpMode=RCIM_Cubic, leaveTangent=-1.625, time=0.7597, value=1.2256}
        {arriveTangent=-4.3961, interpMode=RCIM_Cubic, leaveTangent=-4.3961, tangentMode=RCTM_User, time=1, value=0}
       ]
      }}})
   &DynamicMaterialParameters(
    Index 0 Param 1=1
    Index 0 Param 3=&FloatFromCurve(
     FloatCurve=DataInterface() {propertyValues {Curve {keys [{interpMode=RCIM_Cubic, leaveTangent=8.1085, tangentMode=RCTM_Break, time=0.8568, value=1}, {tangentMode=RCTM_User, time=1, value=0}]}}}
     Scale Curve=0.015
    )
    Index 0 Param 4=0.8
    Index 1 Param 1=&FloatFromCurve(FloatCurve=DataInterface() {propertyValues {Curve {keys [
         {interpMode=RCIM_Cubic, leaveTangent=1.1034, tangentMode=RCTM_Break, time=0, value=0}
         {arriveTangent=2.2399, tangentMode=RCTM_User, time=0.5, value=1}
        ]}}})
    Index 1 Param 2=&FloatFromCurve(FloatCurve=DataInterface() {propertyValues {Curve {keys [<at 0.5 0>, {arriveTangent=3.2053, tangentMode=RCTM_User, time=0.9, value=1}]}}})
    Index 1 Param 3=&FloatFromCurve(FloatCurve=DataInterface() {propertyValues {Curve {keys [<at 0 0 RCIM_Cubic>, {arriveTangent=3.4223, interpMode=RCIM_Cubic, leaveTangent=3.4223, tangentMode=RCTM_User, time=1, value=1}]}}})
    Index 1 Param 4=&FloatFromCurve(
     FloatCurve=DataInterface() {propertyValues {Curve {keys [
         {interpMode=RCIM_Cubic, leaveTangent=0.1169, tangentMode=RCTM_Break, time=0, value=0}
         {arriveTangent=2.5708, interpMode=RCIM_Cubic, leaveTangent=2.5708, tangentMode=RCTM_User, time=1, value=1}
        ]}}}
     Scale Curve=&RandomRangeFloat(Minimum=1, Maximum=40)
    )
    Index 2 Param 1=&FloatFromCurve(FloatCurve=DataInterface() {propertyValues {Curve {keys [
         {interpMode=RCIM_Cubic, leaveTangent=-0.0025, tangentMode=RCTM_Break, time=0, value=0.5}
         {arriveTangent=-0.0316, interpMode=RCIM_Cubic, tangentMode=RCTM_Break, time=1, value=-0.2}
        ]}}})
    Index 2 Param 2=&FloatFromCurve(
     FloatCurve=DataInterface() {propertyValues {Curve {keys [
         {interpMode=RCIM_Cubic, tangentMode=RCTM_User, time=0, value=0}
         {arriveTangent=2.0522, interpMode=RCIM_Cubic, leaveTangent=2.0522, tangentMode=RCTM_User, time=1, value=1}
        ]}}}
     Scale Curve=2
    )
    Module.Index0Float0Write=true
    Module.Index0Float1Write=true
    Module.Index0Float2Write=true
    Module.Index0Float3Write=true
    Module.Index1Float0Write=true
    Module.Index1Float1Write=true
    Module.Index1Float2Write=true
    Module.Index1Float3Write=true
    Module.Index2Float0Write=true
    Module.Index2Float1Write=true
    Module.Index2Float2Write=false
    Module.Index2Float3Write=false
    Module.Index3Float0Write=false
    Module.Index3Float1Write=false
    Module.Index3Float2Write=false
    Module.Index3Float3Write=true
    Write Parameter Index 1=::NewEnumerator1
    Write Parameter Index 2=::NewEnumerator1
    Write Parameter Index 3=::NewEnumerator1
   )
   &ScaleColor(
    Scale Alpha=&FloatFromCurve(
     FloatCurve=DataInterface() {propertyValues {Curve {keys [
         {tangentMode=RCTM_User, time=0, value=1}
         {arriveTangent=-0.0405, interpMode=RCIM_Cubic, leaveTangent=-0.0405, tangentMode=RCTM_User, time=0.7009, value=1}
         {arriveTangent=-8.0089, interpMode=RCIM_Cubic, leaveTangent=-8.0089, tangentMode=RCTM_User, time=1, value=0}
        ]}}}
     Scale Curve=&Multiply_Float(
      A=&FN_FrontEndFloatModifier(Gameplay=1, Front End=0.5)
      B=&FloatFromCurve(
       FloatCurve=DataInterface() {propertyValues {Curve {
          defaultValue=1
          keys [
           {arriveTangent=8.9136, interpMode=RCIM_Cubic, leaveTangent=8.9136, tangentMode=RCTM_User, time=0, value=0}
           {
            arriveTangentWeight=0.1126
            interpMode=RCIM_Cubic
            leaveTangentWeight=0.1126
            tangentMode=RCTM_User
            tangentWeightMode=RCTWM_WeightedBoth
            time=0.3264
            value=1
           }
           {arriveTangent=-4.0894, interpMode=RCIM_Cubic, leaveTangent=-4.0894, tangentMode=RCTM_User, time=1.0009, value=0}
          ]
         }}}
       Scale Curve=&FNTimeOfDayScalar(Night Scale=0.7, Morning Scale=1, Evening Scale=1)
      )
     )
    )
    Scale RGB=&Lerp_Vector(
     A=&VectorFromFloat(Value=&FloatFromCurve(
       FloatCurve=DataInterface() {propertyValues {Curve {keys [<at 0 1>,<at 0.1 1>,<at 0.2 2>,<at 0.3 1>,<at 0.6 1>,<at 0.7 2>,<at 0.8 1>,<at 1 1>]}}}
       Scale Curve=10
      ))
     B=&VectorFromFloat(Value=&RandomRangeFloat(Minimum=0.3, Maximum=5, Evaluation Type=::NewEnumerator1))
     Alpha=&FloatFromCurve(FloatCurve=DataInterface() {propertyValues {Curve {keys [
          {interpMode=RCIM_Cubic, leaveTangent=2.5294, tangentMode=RCTM_Break, time=0, value=0}
          {arriveTangent=-0.2166, interpMode=RCIM_Cubic, tangentMode=RCTM_Break, time=0.4, value=1}
         ]}}})
    )
    ScaleA=true
    ScaleRGB=false
   )
   /SetVariables() {isSetParameters=true}
   /SetVariables(Particles.Glow Dynamic Parameters:Vector4f=<3 1 0 1>) {isSetParameters=true}
  ]
  rendererClasses [&NiagaraSpriteRendererProperties]
  renderers [
   {rendererClass=&NiagaraSpriteRendererProperties}
   {rendererClass=&NiagaraSpriteRendererProperties, rendererIndex=1}
  ]
 }
 Lightning {
  emitterUpdateScript [
   &EmitterState(Life Cycle Mode=::NewEnumerator1, Loop Behavior=::NewEnumerator1, Loop Delay=0.55, Loop Duration=1.6, UseLoopDelay=true)
   &SpawnBurst_Instantaneous(Spawn Count=5, Spawn Group=2, Spawn Time=1.56, Use Spawn Probability=false, UseLoopCountLimit=false) {module=SpawnBurst_Instantaneous001}
   &SpawnBurst_Instantaneous(Spawn Count=20, Spawn Group=1, Use Spawn Probability=false, UseLoopCountLimit=false)
   &SpawnRate(
    SpawnRate=&FloatFromCurve(
     FloatCurve=DataInterface() {propertyValues {Curve {
        defaultValue=1
        keys [
         {arriveTangent=0.1628, interpMode=RCIM_Cubic, leaveTangent=0.1628, tangentMode=RCTM_User, time=0, value=1.364}
         {arriveTangent=-0.853, interpMode=RCIM_Cubic, leaveTangent=-0.853, tangentMode=RCTM_User, time=0.3245, value=0.5806}
         {arriveTangent=0.6788, interpMode=RCIM_Cubic, leaveTangent=0.6788, tangentMode=RCTM_User, time=0.7947, value=0.4558}
         {arriveTangentWeight=0.1237, interpMode=RCIM_Cubic, leaveTangentWeight=0.3709, tangentMode=RCTM_User, time=0.8436, value=1.1289}
         {arriveTangent=0.1123, interpMode=RCIM_Cubic, leaveTangent=0.1123, tangentMode=RCTM_User, time=0.9843, value=0}
        ]
       }}}
     CurveIndex=$Emitter.NormalizedLoopAge
     Scale Curve=40
    )
    Use Spawn Probability=false
   )
  ]
  particleSpawnScript [
   &InitializeParticle(
    Color Mode=::NewEnumerator0
    Lifetime=&MaskFloatBySpawnGroup(
     Masked Float=&RandomRangeFloat(Minimum=0.2, Maximum=0.15)
     Passthrough Float=&MaskFloatBySpawnGroup(
      Masked Float=&RandomRangeFloat(Minimum=0.3, Maximum=0.2)
      Passthrough Float=&RandomRangeFloat(Minimum=0.3, Maximum=0.25)
      Spawn Group Mask=1
     )
    )
    Mesh Renderer Array Visibility Mode=::NewEnumerator3
    Mesh Renderer Info=DataInterface() {propertyValues {MeshRenderer=&self:Lightning_0.NiagaraMeshRendererProperties_1}}
    Mesh Scale=&Multiply_VectorByFloat(
     Vector=&RandomRangeVector(Minimum=<2 1.5 0.8>, Maximum=<1 1 0.5>)
     Float=&MaskFloatBySpawnGroup(Masked Float=5, Passthrough Float=5)
    )
    Mesh Scale Mode=::NewEnumerator4
    Sprite Size Mode=::NewEnumerator1
    Uniform Sprite Size Max=30
    Uniform Sprite Size Min=50
    UsePositionOffset=false
    Write Material Random=true
   )
   &SkeletalMeshLocation(
    Bone / Socket Index=1
    Bone Sampling Mode=::NewEnumerator4
    MaskSpawnGroup=false
    Sampled Position Offset=<0 0 5>
    Skeletal Mesh=$System.Skeletal Mesh
   )
   &ShapeLocation(
    MaskSpawnGroup=false
    Override Seed=false
    Shape Origin=$Particles.SkeletalMeshLocation.SampledPosition
    Sphere Radius=&MaskFloatBySpawnGroup(Masked Float=1, Passthrough Float=&RandomRangeFloat(Minimum=20, Maximum=1), Spawn Group Mask=2)
    Sphere Surface Distribution=1
    UseOwnerScale=true
   )
   &InitialMeshOrientation(
    Mesh Orientation Mode=::NewEnumerator2
    Orientation Vector=$Particles.ShapeLocation.ShapeNormal
    Rotation=&SelectVectorFromArray(
     Array Sampling Mode=::NewEnumerator0
     Direct Array Index=?
     Vector Selection Array=DataInterface() {propertyValues {FloatData [<0 0 0>, <0 0.5 0>]}}
     Random Seed=0
    )
    Use Rotation Vector=true
   )
  ]
  particleUpdateScript [
   &ParticleState()
   &ScaleSpriteSize(
    Uniform Curve Scale=&FloatFromCurve(FloatCurve=DataInterface() {propertyValues {Curve {keys [<at 2.01 1>,<at 2.02 0>]}}}, CurveIndex=$Emitter.Age)
    Uniform Curve Sprite Scale=DataInterface() {propertyValues {Curve {
       defaultValue=0
       keys [
        {interpMode=RCIM_Cubic, tangentMode=RCTM_User, time=0, value=0}
        {interpMode=RCIM_Cubic, tangentMode=RCTM_User, time=0.5, value=1}
        {tangentMode=RCTM_User, time=1, value=0}
       ]
      }}}
   )
   &ScaleMeshSize(Scale Factor=&MaskVectorBySpawnGroup(
     Masked Vector=&VectorFromCurve(
      VectorCurve=DataInterface() {propertyValues {
        XCurve {
         defaultValue=0
         keys [
          {arriveTangent=1.8121, interpMode=RCIM_Cubic, leaveTangent=1.8121, tangentMode=RCTM_User, time=0, value=0}
          {arriveTangent=1.1179, interpMode=RCIM_Cubic, leaveTangent=1.1179, tangentMode=RCTM_User, time=0.3075, value=0.4806}
          {arriveTangent=0.0224, interpMode=RCIM_Cubic, leaveTangent=0.0224, tangentMode=RCTM_User, time=1, value=1}
         ]
        }
        YCurve {
         defaultValue=0
         keys [
          {arriveTangent=1.8121, interpMode=RCIM_Cubic, leaveTangent=1.8121, tangentMode=RCTM_User, time=0, value=0}
          {arriveTangent=1.1179, interpMode=RCIM_Cubic, leaveTangent=1.1179, tangentMode=RCTM_User, time=0.3075, value=0.4806}
          {arriveTangent=0.0224, interpMode=RCIM_Cubic, leaveTangent=0.0224, tangentMode=RCTM_User, time=1, value=1}
         ]
        }
        ZCurve {
         defaultValue=0
         keys [
          {arriveTangent=4.1785, interpMode=RCIM_Cubic, leaveTangent=4.1785, tangentMode=RCTM_User, time=0, value=0}
          {arriveTangent=0.6377, interpMode=RCIM_Cubic, leaveTangent=0.6377, tangentMode=RCTM_User, time=0.2175, value=0.8104}
          {arriveTangent=0.0224, interpMode=RCIM_Cubic, leaveTangent=0.0224, tangentMode=RCTM_User, time=1, value=1}
         ]
        }
       }}
      Scale Curve=<0.07 0.07 0.15>
     )
     Passthrough Vector=&VectorFromFloat(Value=&FloatFromCurve(
       FloatCurve=DataInterface() {propertyValues {Curve {
          defaultValue=1
          keys [
           {arriveTangent=2.8632, interpMode=RCIM_Cubic, leaveTangent=2.8632, tangentMode=RCTM_User, time=0, value=0}
           {arriveTangent=1.1179, interpMode=RCIM_Cubic, leaveTangent=1.1179, tangentMode=RCTM_User, time=0.2042, value=0.4642}
           {arriveTangent=0.0224, interpMode=RCIM_Cubic, leaveTangent=0.0224, tangentMode=RCTM_User, time=1, value=1}
          ]
         }}}
       Scale Curve=0.1
      ))
     Spawn Group Mask=2
    ))
   &DynamicMaterialParameters(
    Index 0 Param 1=&MakeFloatFromVector(Vector=$Particles.Scale, Channel=::NewEnumerator0)
    Index 0 Param 2=&RandomRangeFloat(Minimum=3, Maximum=7)
    Index 0 Param 3=&FloatFromCurve(
     FloatCurve=DataInterface() {propertyValues {Curve {keys [{interpMode=RCIM_Cubic, leaveTangent=8.1085, tangentMode=RCTM_Break, time=0.8568, value=1}, {tangentMode=RCTM_User, time=1, value=0}]}}}
     Scale Curve=0.015
    )
    Index 0 Param 4=&RandomRangeFloat(Minimum=0.2, Maximum=0.4)
    Index 1 Param 1=&FloatFromCurve(FloatCurve=DataInterface() {propertyValues {Curve {keys [
         {interpMode=RCIM_Cubic, leaveTangent=1.1034, tangentMode=RCTM_Break, time=0, value=0}
         {arriveTangent=2.2399, tangentMode=RCTM_User, time=0.5, value=1}
        ]}}})
    Index 1 Param 2=&FloatFromCurve(FloatCurve=DataInterface() {propertyValues {Curve {keys [<at 0.5 0>, {arriveTangent=3.2053, tangentMode=RCTM_User, time=0.9, value=1}]}}})
    Index 1 Param 3=&FloatFromCurve(FloatCurve=DataInterface() {propertyValues {Curve {keys [
         <at 0 0 RCIM_Cubic>
         {arriveTangent=0.5046, interpMode=RCIM_Cubic, leaveTangent=0.5046, time=0.4761, value=0.1221}
         {arriveTangent=-0.0289, interpMode=RCIM_Cubic, leaveTangent=-0.0289, tangentMode=RCTM_User, time=1, value=0.5046}
        ]}}})
    Index 1 Param 4=&FloatFromCurve(
     FloatCurve=DataInterface() {propertyValues {Curve {keys [
         {interpMode=RCIM_Cubic, leaveTangent=0.1169, tangentMode=RCTM_Break, time=0, value=0}
         {arriveTangent=2.5708, interpMode=RCIM_Cubic, leaveTangent=2.5708, tangentMode=RCTM_User, time=1, value=1}
        ]}}}
     Scale Curve=&RandomRangeFloat(Minimum=1, Maximum=40)
    )
    Index 2 Param 1=&FloatFromCurve(FloatCurve=DataInterface() {propertyValues {Curve {keys [
         {interpMode=RCIM_Cubic, leaveTangent=-0.0025, tangentMode=RCTM_Break, time=0, value=0.5}
         {arriveTangent=-0.0316, interpMode=RCIM_Cubic, tangentMode=RCTM_Break, time=1, value=-0.2}
        ]}}})
    Index 2 Param 2=&FloatFromCurve(
     FloatCurve=DataInterface() {propertyValues {Curve {keys [
         {interpMode=RCIM_Cubic, tangentMode=RCTM_User, time=0, value=0}
         {arriveTangent=2.0522, interpMode=RCIM_Cubic, leaveTangent=2.0522, tangentMode=RCTM_User, time=1, value=1}
        ]}}}
     Scale Curve=3
    )
    Module.Index0Float0Write=true
    Module.Index0Float1Write=true
    Module.Index0Float2Write=true
    Module.Index0Float3Write=true
    Module.Index1Float0Write=true
    Module.Index1Float1Write=true
    Module.Index1Float2Write=true
    Module.Index1Float3Write=true
    Module.Index2Float0Write=true
    Module.Index2Float1Write=true
    Module.Index2Float2Write=false
    Module.Index2Float3Write=false
    Module.Index3Float0Write=false
    Module.Index3Float1Write=false
    Module.Index3Float2Write=false
    Module.Index3Float3Write=true
    Write Parameter Index 1=::NewEnumerator1
    Write Parameter Index 2=::NewEnumerator1
    Write Parameter Index 3=::NewEnumerator1
   )
   &ScaleColor(
    Scale Alpha=&FloatFromCurve(
     FloatCurve=DataInterface() {propertyValues {Curve {keys [
         {tangentMode=RCTM_User, time=0, value=1}
         {arriveTangent=-0.0405, interpMode=RCIM_Cubic, leaveTangent=-0.0405, tangentMode=RCTM_User, time=0.3819, value=1}
         {arriveTangent=-4.2694, interpMode=RCIM_Cubic, leaveTangent=-4.2694, tangentMode=RCTM_User, time=1, value=0}
        ]}}}
     Scale Curve=&FN_FrontEndFloatModifier(Gameplay=1, Front End=0.5)
    )
    Scale RGB=&Lerp_Vector(
     A=&MakeVectorFromLinearColorRGB(Color=&Multiply_LinearColorByFloat(Linear Color=$User.LightningColor, Float=1.5))
     B=&MakeVectorFromLinearColorRGB(Color=&SelectLinearColorFromArray(
       Array Sampling Mode=::NewEnumerator1
       Direct Array Index=$Engine.ExecIndex
       Direct Array Mode=::NewEnumerator1
       Color Selection Array=DataInterface() {propertyValues {ColorData [<rgba 5 4.6482 0.6567 1>, <rgba 0.0274 5 1.1324 1>, <rgba 1.0025 2.8799 5 1>, <rgba 0.5806 0.5514 5 1>]}}
       Random Seed=?
      ))
     Alpha=&FloatFromCurve(
      FloatCurve=DataInterface() {propertyValues {Curve {
         defaultValue=1
         keys [{interpMode=RCIM_Cubic, tangentMode=RCTM_User, time=1.5528, value=0}, {interpMode=RCIM_Cubic, tangentMode=RCTM_User, time=1.5716, value=1}]
        }}}
      CurveIndex=$Emitter.Age
     )
    )
    ScaleA=true
    ScaleRGB=true
   )
   &CameraOffset(Camera Offset Amount=25)
  ]
  rendererClasses [&NiagaraMeshRendererProperties, &NiagaraSpriteRendererProperties]
  renderers [
   {rendererClass=&NiagaraMeshRendererProperties, rendererIndex=0}
   {rendererClass=&NiagaraMeshRendererProperties, rendererIndex=1}
   {rendererClass=&NiagaraSpriteRendererProperties, rendererIndex=2}
  ]
 }
 Lightning001 {
  emitterUpdateScript [
   &EmitterState(Life Cycle Mode=::NewEnumerator1, Loop Behavior=::NewEnumerator1, Loop Delay=0.15, Loop Duration=0.25, UseLoopDelay=true)
   &SpawnRate(SpawnRate=20, Use Spawn Probability=false)
  ]
  particleSpawnScript [
   &InitializeParticle(
    Color Mode=::NewEnumerator0
    Lifetime=&MaskFloatBySpawnGroup(
     Masked Float=&RandomRangeFloat(Minimum=0.3, Maximum=0.07)
     Passthrough Float=&RandomRangeFloat(Minimum=0.3, Maximum=0.2)
    )
    Mesh Scale=&Multiply_VectorByFloat(Vector=&RandomRangeVector(Minimum=<2 1.5 0.8>, Maximum=<2 2 1>), Float=2)
    Mesh Scale Mode=::NewEnumerator4
    UsePositionOffset=false
    Write Material Random=true
   )
   &SkeletalMeshLocation(Bone Sampling Mode=::NewEnumerator4, MaskSpawnGroup=false, Skeletal Mesh=$System.Skeletal Mesh)
   &ShapeLocation(
    MaskSpawnGroup=false
    Offset=<0 2 0>
    Offset Mode=::NewEnumerator0
    Override Seed=false
    Shape Origin=$Particles.SkeletalMeshLocation.SampledPosition
    Sphere Radius=0.5
    UseOwnerScale=true
   )
   &InitialMeshOrientation(
    Mesh Orientation Mode=::NewEnumerator2
    Orientation Vector=$Particles.ShapeLocation.ShapeNormal
    Rotation=&SelectVectorFromArray(
     Array Sampling Mode=::NewEnumerator0
     Direct Array Index=?
     Vector Selection Array=DataInterface() {propertyValues {FloatData [<0 0 0>, <0 0.5 0>]}}
     Random Seed=0
    )
    Use Rotation Vector=true
   )
  ]
  particleUpdateScript [
   &SolveForcesAndVelocity(Clamp Velocity=false, Limit Acceleration=false)
   &ParticleState()
   &ScaleMeshSize(Scale Factor=&VectorFromFloat(Value=&FloatFromCurve(
      FloatCurve=DataInterface() {propertyValues {Curve {
         defaultValue=1
         keys [
          {arriveTangent=4.1785, interpMode=RCIM_Cubic, leaveTangent=4.1785, tangentMode=RCTM_User, time=0, value=0}
          {arriveTangent=1.1179, interpMode=RCIM_Cubic, leaveTangent=1.1179, tangentMode=RCTM_User, time=0.2042, value=0.6228}
          {arriveTangent=0.0224, interpMode=RCIM_Cubic, leaveTangent=0.0224, tangentMode=RCTM_User, time=1, value=1}
         ]
        }}}
      Scale Curve=0.1
     )))
   &DynamicMaterialParameters(
    Index 0 Param 1=&MakeFloatFromVector(Vector=$Particles.Scale, Channel=::NewEnumerator0)
    Index 0 Param 2=&RandomRangeFloat(Minimum=5, Maximum=10)
    Index 0 Param 3=&FloatFromCurve(
     FloatCurve=DataInterface() {propertyValues {Curve {keys [{interpMode=RCIM_Cubic, leaveTangent=8.1085, tangentMode=RCTM_Break, time=0.8568, value=1}, {tangentMode=RCTM_User, time=1, value=0}]}}}
     Scale Curve=0.015
    )
    Index 0 Param 4=1
    Index 1 Param 1=&FloatFromCurve(FloatCurve=DataInterface() {propertyValues {Curve {keys [
         {interpMode=RCIM_Cubic, leaveTangent=1.1034, tangentMode=RCTM_Break, time=0, value=0}
         {arriveTangent=2.2399, tangentMode=RCTM_User, time=0.5, value=1}
        ]}}})
    Index 1 Param 2=&FloatFromCurve(FloatCurve=DataInterface() {propertyValues {Curve {keys [<at 0.5 0>, {arriveTangent=3.2053, tangentMode=RCTM_User, time=0.9, value=1}]}}})
    Index 1 Param 3=&FloatFromCurve(FloatCurve=DataInterface() {propertyValues {Curve {keys [
         <at 0 0 RCIM_Cubic>
         {arriveTangent=0.5046, interpMode=RCIM_Cubic, leaveTangent=0.5046, time=0.8846, value=0.0963}
         {arriveTangent=2.773, interpMode=RCIM_Cubic, leaveTangent=2.773, tangentMode=RCTM_User, time=1, value=0.5046}
        ]}}})
    Index 1 Param 4=&FloatFromCurve(
     FloatCurve=DataInterface() {propertyValues {Curve {keys [
         {interpMode=RCIM_Cubic, leaveTangent=0.1169, tangentMode=RCTM_Break, time=0, value=0}
         {arriveTangent=2.5708, interpMode=RCIM_Cubic, leaveTangent=2.5708, tangentMode=RCTM_User, time=1, value=1}
        ]}}}
     Scale Curve=&RandomRangeFloat(Minimum=1, Maximum=40)
    )
    Index 2 Param 1=&FloatFromCurve(FloatCurve=DataInterface() {propertyValues {Curve {keys [
         {interpMode=RCIM_Cubic, leaveTangent=-0.0025, tangentMode=RCTM_Break, time=0, value=0.5}
         {arriveTangent=-0.0316, interpMode=RCIM_Cubic, tangentMode=RCTM_Break, time=1, value=-0.2}
        ]}}})
    Index 2 Param 2=&FloatFromCurve(
     FloatCurve=DataInterface() {propertyValues {Curve {keys [
         {interpMode=RCIM_Cubic, tangentMode=RCTM_User, time=0, value=0}
         {arriveTangent=2.0522, interpMode=RCIM_Cubic, leaveTangent=2.0522, tangentMode=RCTM_User, time=1, value=1}
        ]}}}
     Scale Curve=3
    )
    Module.Index0Float0Write=true
    Module.Index0Float1Write=true
    Module.Index0Float2Write=true
    Module.Index0Float3Write=true
    Module.Index1Float0Write=true
    Module.Index1Float1Write=true
    Module.Index1Float2Write=true
    Module.Index1Float3Write=true
    Module.Index2Float0Write=true
    Module.Index2Float1Write=true
    Module.Index2Float2Write=false
    Module.Index2Float3Write=false
    Module.Index3Float0Write=false
    Module.Index3Float1Write=false
    Module.Index3Float2Write=false
    Module.Index3Float3Write=true
    Write Parameter Index 1=::NewEnumerator1
    Write Parameter Index 2=::NewEnumerator1
    Write Parameter Index 3=::NewEnumerator1
   )
   &ScaleColor(
    Scale Alpha=&FloatFromCurve(
     FloatCurve=DataInterface() {propertyValues {Curve {keys [
         {tangentMode=RCTM_User, time=0, value=1}
         {arriveTangent=-0.0405, interpMode=RCIM_Cubic, leaveTangent=-0.0405, tangentMode=RCTM_User, time=0.7009, value=1}
         {arriveTangent=-8.0089, interpMode=RCIM_Cubic, leaveTangent=-8.0089, tangentMode=RCTM_User, time=1, value=0}
        ]}}}
     Scale Curve=&FN_FrontEndFloatModifier(Gameplay=1, Front End=0.5)
    )
    Scale RGB=&MakeVectorFromLinearColorRGB(Color=$User.LightningColor)
    ScaleA=true
    ScaleRGB=true
   )
  ]
  rendererClasses [&NiagaraMeshRendererProperties]
  renderers [
   {rendererClass=&NiagaraMeshRendererProperties, rendererIndex=0}
   {rendererClass=&NiagaraMeshRendererProperties, rendererIndex=1}
  ]
 }
 Main_RingCombined {
  emitterUpdateScript [
   &EmitterState(Life Cycle Mode=::NewEnumerator1, Loop Behavior=::NewEnumerator1, Loop Delay=0.2, Loop Duration=3, UseLoopDelay=true)
   &SpawnBurst_Instantaneous(Spawn Group=2, Spawn Time=1.81, Use Spawn Probability=false, UseLoopCountLimit=false) {module=SpawnBurst_Instantaneous002}
   &SpawnBurst_Instantaneous(Spawn Group=1, Spawn Time=0.35, Use Spawn Probability=false, UseLoopCountLimit=false) {module=SpawnBurst_Instantaneous001}
  ]
  particleSpawnScript [
   /SetVariables(Particles.RandomRange:NiagaraFloat=&Multiply_Float(
     A=&RandomRangeFloat(Minimum=0.25, Maximum=0.375)
     B=&FloatFromCurve(
      FloatCurve=DataInterface() {propertyValues {Curve {keys [<at 0 1 RCIM_Cubic>, {arriveTangent=7.3942, interpMode=RCIM_Cubic, leaveTangent=7.3942, tangentMode=RCTM_User, time=0.3, value=2}]}}}
      CurveIndex=$Emitter.Age
     )
    )) {isSetParameters=true}
   &InitializeParticle(
    Color=&FNTimeOfDayColor(
     Night Color=<rgba 0.8 0.9367 1 0.6>
     Morning Color=<rgba 0.5958 0.6976 0.7448 0.6>
     Day Color=<rgba 0.8 0.9367 1 0.6>
     Evening Color=<rgba 0.575 0.6732 0.7188 0.6>
    )
    Lifetime=&MaskFloatBySpawnGroup(Masked Float=0.3, Passthrough Float=0.45)
    Mesh Scale=&Multiply_VectorByFloat(Vector=<1 1.5 1>, Float=0.16)
    Mesh Scale Mode=::NewEnumerator4
    UsePositionOffset=false
    Write Material Random=true
   )
   &SystemLocation(MaskSpawnGroup=false, Offset=<0 7.0031 0>)
   &CameraOffset(Camera Offset Amount=15)
   &SkeletalMeshLocation(
    Bone / Socket Index=&MaskIntBySpawnGroup(Masked Int=0, Passthrough Int=1, Spawn Group Mask=0)
    Bone Sampling Mode=::NewEnumerator4
    MaskSpawnGroup=false
    Skeletal Mesh=$System.Skeletal Mesh
   )
  ]
  particleUpdateScript [
   &ParticleState()
   &ScaleMeshSize(Scale Factor=&VectorFromFloat(Value=&FloatFromCurve(
      FloatCurve=DataInterface() {propertyValues {Curve {keys [
          {arriveTangent=5.6347, interpMode=RCIM_Cubic, leaveTangent=5.6347, tangentMode=RCTM_User, time=0, value=0.3788}
          {arriveTangent=1.3821, interpMode=RCIM_Cubic, leaveTangent=1.3821, time=0.1115, value=0.9411}
          {arriveTangent=0.2133, interpMode=RCIM_Cubic, leaveTangent=0.2133, tangentMode=RCTM_User, time=0.5902, value=1.1945}
          <at 0.9985 1.2296 RCIM_Cubic>
         ]}}}
      Scale Curve=&MaskFloatBySpawnGroup(
       Masked Float=0.95
       Passthrough Float=&MaskFloatBySpawnGroup(Masked Float=1.6, Passthrough Float=1.5, Spawn Group Mask=1)
      )
     )))
   &DynamicMaterialParameters(
    Index 0 Param 1=&FloatFromCurve(FloatCurve=DataInterface() {propertyValues {Curve {keys [
         {arriveTangent=2.3202, interpMode=RCIM_Cubic, leaveTangent=2.3202, tangentMode=RCTM_User, time=0, value=0.2481}
         {arriveTangent=0.6377, interpMode=RCIM_Cubic, leaveTangent=0.6377, time=0.0995, value=0.4178}
         {interpMode=RCIM_Cubic, tangentMode=RCTM_User, time=0.3951, value=0.5}
         {arriveTangent=-1.1781, tangentMode=RCTM_User, time=1, value=0}
        ]}}})
    Module.Index0Float0Write=true
    Module.Index0Float1Write=true
    Module.Index0Float2Write=true
    Module.Index0Float3Write=true
   )
   &ScaleColor(
    Scale Alpha=&FloatFromCurve(
     FloatCurve=DataInterface() {propertyValues {Curve {keys [
         {arriveTangent=-0.0871, interpMode=RCIM_Cubic, leaveTangent=-0.0871, tangentMode=RCTM_User, time=0, value=0.6231}
         {interpMode=RCIM_Cubic, tangentMode=RCTM_User, time=0.5, value=0.7329}
         {tangentMode=RCTM_User, time=1, value=0}
        ]}}}
     Scale Curve=0.4
    )
    Scale RGB=&MakeVectorFromLinearColorRGB(Color=&MaskLinearColorBySpawnGroup(Masked Linear Color=<rgba 1 0.7152 0.0296 1>, Passthrough Linear Color=$User.LightningColor, Spawn Group Mask=2))
    ScaleA=true
    ScaleRGB=true
   )
   &UpdateMeshOrientation(Rotation Rate=0.2, Rotation Vector=<0 1 0>) {module=UpdateMeshOrientation001}
  ]
  rendererClasses [&NiagaraMeshRendererProperties]
  renderers [{rendererClass=&NiagaraMeshRendererProperties, rendererIndex=0}]
 }
 ParticleDecal {
  emitterUpdateScript [
   &EmitterState(Life Cycle Mode=::NewEnumerator1, Loop Behavior=::NewEnumerator1, Loop Delay=0.41, Loop Duration=3, UseLoopDelay=true)
   &SpawnParticlesFromOtherEmitter(
    Attribute Reader=DataInterface() {propertyValues {EmitterBinding {bindingMode=Other, emitterName=Lightning}}}
    Spawn Rate=&FloatFromCurve(
     FloatCurve=DataInterface() {propertyValues {Curve {
        defaultValue=1
        keys [
         {interpMode=RCIM_Cubic, tangentMode=RCTM_User, time=1.4629, value=1}
         {interpMode=RCIM_Cubic, tangentMode=RCTM_User, time=1.5246, value=0.1697}
        ]
       }}}
     CurveIndex=$Emitter.Age
     Scale Curve=10
    )
    Use Maximum Age=false
    Use Minimum Age=false
   )
  ]
  particleSpawnScript [
   &InitializeParticle(
    Lifetime=0.2
    Mesh Scale Mode=::NewEnumerator3
    Mesh Uniform Scale=2
    Position Offset=<0 0 100>
    Sprite Size Mode=::NewEnumerator3
    Uniform Sprite Size=50
    UsePositionOffset=false
    Write Material Random=true
   )
   &CameraOffset(Camera Offset Amount=10)
   &DynamicMaterialParameters(
    Index 0 Param 1=1
    Index 0 Param 3=1
    Module.Index0Param0Write=true
    Module.Index0Param1Write=true
    Module.Index0Param2Write=true
    Module.Index0Param3Write=true
   )
   &SampleParticlesFromOtherEmitter(Lifetime Sampling:ENiagara_DirectReadSamplingMode=::NewEnumerator0)
  ]
  particleUpdateScript [
   &ParticleState()
   &ScaleColor(
    Scale Alpha=&FloatFromCurve(
     FloatCurve=DataInterface() {propertyValues {Curve {
        defaultValue=1
        keys [
         {interpMode=RCIM_Cubic, tangentMode=RCTM_User, time=0.1247, value=1}
         {arriveTangent=-0.2743, interpMode=RCIM_Cubic, leaveTangent=-0.2743, tangentMode=RCTM_User, time=0.5942, value=0.1526}
         {arriveTangent=-0.5414, interpMode=RCIM_Cubic, leaveTangent=-0.5414, tangentMode=RCTM_User, time=1, value=0}
        ]
       }}}
     Scale Curve=0.5
    )
    Scale RGB=&MakeVectorFromLinearColorRGB(Color=&LerpLinearColors(
      StartColor=$User.LightningColor
      EndColor=<rgba 1 0.7203 0 1>
      LerpFactor=&FloatFromCurve(
       FloatCurve=DataInterface() {propertyValues {Curve {
          defaultValue=1
          keys [{interpMode=RCIM_Cubic, tangentMode=RCTM_User, time=1.571, value=0}, {interpMode=RCIM_Cubic, tangentMode=RCTM_User, time=1.5806, value=1}]
         }}}
       CurveIndex=$Emitter.Age
      )
     ))
    ScaleA=true
    ScaleRGB=true
   )
   &SolveForcesAndVelocity(Clamp Velocity=false, Limit Acceleration=false)
  ]
  rendererClasses [&NiagaraSpriteRendererProperties]
  renderers [{rendererClass=&NiagaraSpriteRendererProperties}]
 }
 ParticleDecal001 {
  emitterUpdateScript [
   &EmitterState(Life Cycle Mode=::NewEnumerator1, Loop Behavior=::NewEnumerator1, Loop Delay=0.41, Loop Duration=3, UseLoopDelay=false)
   &SpawnParticlesFromOtherEmitter(
    Attribute Reader=DataInterface() {propertyValues {EmitterBinding {bindingMode=Other, emitterName=CircularLightning}}}
    Calculate Spawn Rate Per Particle=false
    Spawn Rate=30
    Use Maximum Age=false
    Use Minimum Age=false
   )
  ]
  particleSpawnScript [
   &PartitionParticles(Number Of Partitions=20, Particles Per Partition=2, Partition Mode:ENiagara_PartitionMode_v2=::NewEnumerator1)
   &InitializeParticle(
    Color=&Multiply_LinearColorByFloat(Linear Color=$User.LightningColor, Float=1.2)
    Lifetime=0.15
    Mesh Scale Mode=::NewEnumerator3
    Mesh Uniform Scale=2
    Position Offset=<0 0 100>
    Sprite Size Mode=::NewEnumerator3
    Uniform Sprite Size=30
    UsePositionOffset=false
    Write Material Random=true
   )
   &CameraOffset(Camera Offset Amount=10)
   &DynamicMaterialParameters(
    Index 0 Param 1=3
    Index 0 Param 3=1
    Module.Index0Float0Write=true
    Module.Index0Float1Write=true
    Module.Index0Float2Write=true
    Module.Index0Float3Write=true
   )
   &SampleParticlesFromOtherEmitter(
    Age Sampling:ENiagara_DirectReadSamplingMode=::NewEnumerator0
    Lifetime Sampling:ENiagara_DirectReadSamplingMode=::NewEnumerator0
    Mesh Orientation Sampling:ENiagara_DirectReadSamplingMode=::NewEnumerator0
    Position Sampling:ENiagara_DirectReadSamplingMode=::NewEnumerator1
   )
   /SetVariables(Particles.OffsetPosition:NiagaraPosition=&AddVectorToPosition(
     Position=$Output.SampleParticlesFromOtherEmitter.SampledPosition
     Vector=&Multiply_VectorByFloat(
      Vector=&ReturnMeshOrientationAxis(Mesh Orientation Axis=::YAxis, Mesh Orientation Quat=$Particles.MeshOrientation)
      Float=&MaskFloat(Mask Source=1, Mask Parameter=$Particles.PartitionParticles.Index In Partition, Mask=1, Masked Float=-15, Passthrough Float=15)
     )
    )) {isSetParameters=true}
  ]
  particleUpdateScript [
   &ParticleState()
   &ScaleColor(
    Scale Alpha=&FloatFromCurve(FloatCurve=DataInterface() {propertyValues {Curve {
        defaultValue=1
        keys [
         {interpMode=RCIM_Cubic, tangentMode=RCTM_User, time=0.1247, value=1}
         {arriveTangent=-0.2743, interpMode=RCIM_Cubic, leaveTangent=-0.2743, tangentMode=RCTM_User, time=0.5942, value=0.1526}
         {arriveTangent=-0.5414, interpMode=RCIM_Cubic, leaveTangent=-0.5414, tangentMode=RCTM_User, time=1, value=0}
        ]
       }}})
    ScaleA=true
    ScaleRGB=false
   )
   &SolveForcesAndVelocity(Clamp Velocity=false, Limit Acceleration=false)
  ]
  rendererClasses [&NiagaraSpriteRendererProperties]
  renderers [{rendererClass=&NiagaraSpriteRendererProperties}]
 }
 Vortex {
  emitterUpdateScript [
   &EmitterState(Inactive Response=::NewEnumerator1, Life Cycle Mode=::NewEnumerator1, Loop Behavior=::NewEnumerator1, Loop Duration=3, UseLoopDelay=false)
   &SpawnBurst_Instantaneous(Spawn Time=1.77, Use Spawn Probability=false, UseLoopCountLimit=false)
   &SpawnBurst_Instantaneous(Spawn Group=1, Spawn Time=0.37, Use Spawn Probability=false, UseLoopCountLimit=false) {module=SpawnBurst_Instantaneous001}
   &SpawnBurst_Instantaneous(Spawn Group=2, Spawn Time=1.55, Use Spawn Probability=false, UseLoopCountLimit=false) {module=SpawnBurst_Instantaneous002}
   /SetVariables(Emitter.Erosion:NiagaraFloat=&FloatFromCurve(
     FloatCurve=DataInterface() {propertyValues {Curve {keys [<at 0.3914 0.3 RCIM_Cubic>,<at 1.0907 0.8 RCIM_Cubic>]}}}
     CurveIndex=$Emitter.NormalizedLoopAge
    )) {isSetParameters=true}
  ]
  particleSpawnScript [
   &InitializeParticle(
    Color=<rgba 3 3 3 2>
    Lifetime=&MaskFloatBySpawnGroup(Masked Float=0.35, Passthrough Float=0.605)
    Mesh Scale Mode=::NewEnumerator3
    Mesh Uniform Scale=2.4
    Position Offset=<0 0 5>
    Sprite Rotation Angle=90
    Sprite Rotation Mode=::NewEnumerator0
    Sprite Size Mode=::NewEnumerator3
    Sprite UV Mode=::NewEnumerator4
    Uniform Sprite Size=150
    UsePositionOffset=true
    Write Material Random=true
   )
   &SkeletalMeshLocation(Bone / Socket Index=1, Bone Sampling Mode=::NewEnumerator4, MaskSpawnGroup=false, Skeletal Mesh=$System.Skeletal Mesh)
   &CameraOffset(Camera Offset Amount=&MaskFloatBySpawnGroup(Masked Float=20, Passthrough Float=0))
   &InitialMeshOrientation(Mesh Axis To Orient=<0 0 1>, Mesh Orientation Mode=::NewEnumerator2, Orientation Vector=<0 1 0>, Use Rotation Vector=false)
   /SetVariables(Particles.VisibilityTag=&SpawnGroup()) {isSetParameters=true}
  ]
  particleUpdateScript [
   &ParticleState() {module=ParticleState001}
   &UpdateMeshOrientation(Rotation Coordinate Space=::NewEnumerator0, Rotation Rate=10, Rotation Vector=<0 1 0>)
   &DynamicMaterialParameters(
    Index 0 Param 1=1
    Index 0 Param 3=1
    Index 0 Param 4=1
    Module.Index0Float0Write=true
    Module.Index0Float1Write=true
    Module.Index0Float2Write=true
    Module.Index0Float3Write=true
   )
   &ScaleColor(
    Scale Alpha=&FloatFromCurve(
     FloatCurve=DataInterface() {propertyValues {Curve {keys [
         {interpMode=RCIM_Cubic, tangentMode=RCTM_User, time=0, value=0}
         {interpMode=RCIM_Cubic, tangentMode=RCTM_User, time=0.6778, value=1}
         {arriveTangentWeight=0.2755, interpMode=RCIM_Cubic, leaveTangentWeight=0.1907, tangentMode=RCTM_User, time=0.985, value=1}
         {arriveTangent=-2.5571, tangentMode=RCTM_User, time=1, value=0}
        ]}}}
     Scale Curve=&Multiply_Float(A=1, B=&FNTimeOfDayScalar(Night Scale=0.3, Morning Scale=0.5, Day Scale=0.5, Evening Scale=0.5))
    )
    Scale RGB=&MakeVectorFromLinearColorRGB(Color=<rgba 0.4399 0.0499 1 1>)
    ScaleA=true
    ScaleRGB=false
   )
   /SetVariables(Particles.VortexScale:Vector3f=&Multiply_VectorByFloat(
     Vector=<1 1 1>
     Float=&FloatFromCurve(
      FloatCurve=DataInterface() {propertyValues {Curve {keys [
          {
           arriveTangent=-2.8682
           arriveTangentWeight=0.1242
           interpMode=RCIM_Cubic
           leaveTangent=-2.8682
           leaveTangentWeight=0.0922
           tangentMode=RCTM_User
           time=0.0573
           value=1.5945
          }
          {
           arriveTangent=-1.1704
           arriveTangentWeight=0.1427
           interpMode=RCIM_Cubic
           leaveTangent=-1.1704
           leaveTangentWeight=0.0397
           tangentMode=RCTM_User
           time=0.7239
           value=0.6518
          }
          {arriveTangent=-5.9456, interpMode=RCIM_Cubic, leaveTangent=-5.9456, tangentMode=RCTM_User, time=1, value=0}
         ]}}}
      Scale Curve=0.5
     )
    )) {isSetParameters=true}
   &Light_Attributes(
    Color=1
    Light Radius=200
    Radius=1
    Scale Alpha:NiagaraFloat=&FloatFromCurve(
     FloatCurve=DataInterface() {propertyValues {Curve {keys [
         {tangentMode=RCTM_User, time=0, value=1}
         {arriveTangent=-0.0405, interpMode=RCIM_Cubic, leaveTangent=-0.0405, tangentMode=RCTM_User, time=0.7009, value=1}
         {arriveTangent=-8.0089, interpMode=RCIM_Cubic, leaveTangent=-8.0089, tangentMode=RCTM_User, time=1, value=0}
        ]}}}
     Scale Curve=&Multiply_Float(
      A=&FN_FrontEndFloatModifier(Gameplay=1, Front End=0.5)
      B=&FloatFromCurve(FloatCurve=DataInterface() {propertyValues {Curve {
          defaultValue=1
          keys [
           {arriveTangent=8.9136, interpMode=RCIM_Cubic, leaveTangent=8.9136, tangentMode=RCTM_User, time=0, value=0}
           {
            arriveTangentWeight=0.1126
            interpMode=RCIM_Cubic
            leaveTangentWeight=0.1126
            tangentMode=RCTM_User
            tangentWeightMode=RCTWM_WeightedBoth
            time=0.3264
            value=1
           }
           {arriveTangent=-4.0894, interpMode=RCIM_Cubic, leaveTangent=-4.0894, tangentMode=RCTM_User, time=1.0009, value=0}
          ]
         }}})
     )
    )
    Scale RGB:Vector3f=&MakeVectorFromLinearColorRGB(Color=&MaskLinearColorBySpawnGroup(
      Masked Linear Color=<rgba 5 4.0555 1.6233 1>
      Passthrough Linear Color=&Multiply_LinearColorByFloat(Linear Color=$User.LightningColor, Float=2)
      Spawn Group Mask=2
     ))
    ScaleA=true
    ScaleRGB=true
   )
  ]
  rendererClasses [&NiagaraMeshRendererProperties, &NiagaraLightRendererProperties]
  renderers [
   {rendererClass=&NiagaraMeshRendererProperties, rendererIndex=0}
   {rendererClass=&NiagaraLightRendererProperties, rendererIndex=1}
   {rendererClass=&NiagaraLightRendererProperties, rendererIndex=2}
  ]
 }
}
summary {
 systemName=NS_Emote_FumeFleeceClap_Lightning
 emitters [
  {emitterName=Flash, rendererClasses [&NiagaraSpriteRendererProperties]}
  {emitterName=Lightning, rendererClasses [&NiagaraMeshRendererProperties, &NiagaraSpriteRendererProperties]}
  {emitterName=Main_RingCombined, rendererClasses [&NiagaraMeshRendererProperties]}
  {emitterName=Vortex, rendererClasses [&NiagaraMeshRendererProperties, &NiagaraLightRendererProperties]}
  {emitterName=CircularLightning, rendererClasses [&NiagaraMeshRendererProperties]}
  {emitterName=Lightning001, rendererClasses [&NiagaraMeshRendererProperties]}
  {emitterName=ParticleDecal, rendererClasses [&NiagaraSpriteRendererProperties]}
  {emitterName=ParticleDecal001, rendererClasses [&NiagaraSpriteRendererProperties]}
 ]
 userVariables [
  User.bIsFrontEnd {type=NiagaraBool}
  User.bIsFrontEndPreview {type=NiagaraBool}
  User.LightningColor {type=LinearColor, description="", defaultValue {struct=LinearColor, value=<rgba 0 0.3269 3 1>}}
 ]
}
