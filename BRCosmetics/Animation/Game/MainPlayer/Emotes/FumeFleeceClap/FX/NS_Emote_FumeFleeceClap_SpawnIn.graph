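# Niagara spawn-in effect for the FumeFleeceClap emote: a short lightning burst on the skeletal mesh, colored by User.LightningColor (blue default).
# Staggered one-shot bursts layer a lightning mesh, decal and light sprites, and dragged, speed-stretched sparks, all sampling a mesh bone or socket.
# Flash sprites fire twice, with spawn groups masking socket, color and size.
# Curves drive dynamic material parameters and fades, with Flash alpha also scaled for front end and time of day.
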NiagaraSystem /BRCosmetics/Animation/Game/MainPlayer/Emotes/FumeFleeceClap/FX/NS_Emote_FumeFleeceClap_SpawnIn
dependencies {
 usedDataInterfaces [&NiagaraDataInterfaceCurve, &NiagaraDataInterfaceColorCurve, &NiagaraDataInterfaceSkeletalMesh]
 usedDynamicInputs [
  &MakeCustomVectorFromBool
  &SetBoolByFloatComparison
  &MakeLinearColorFromVectorAndFloat
  &MakeVectorFromLinearColorRGB
 ]
 usedModules [
  &SystemState
  &self:SystemSpawnScript.SystemScriptSource.SystemScriptGraph.NiagaraNodeAssignment_0.SetVariables_#1
 ]
}
emitters {
 Flash {
  emitterUpdateScript [
   &EmitterState(Life Cycle Mode=::NewEnumerator1, Loop Behavior=::NewEnumerator1, Loop Duration=3, UseLoopDelay=false)
   &SpawnBurst_Instantaneous(Spawn Group=1, Spawn Time=1.99, Use Spawn Probability=false, UseLoopCountLimit=false) {module=SpawnBurst_Instantaneous001}
   &SpawnBurst_Instantaneous(Use Spawn Probability=false, UseLoopCountLimit=false)
  ]
  particleSpawnScript [
   &InitializeParticle(
    Color=&Multiply_LinearColorByFloat(
     Linear Color=&MaskLinearColorBySpawnGroup(Masked Linear Color=$User.LightningColor, Passthrough Linear Color=<rgba 3 2.4333 0.974 1>)
     Float=1.5
    )
    Lifetime=&MaskFloatBySpawnGroup(Masked Float=0.4, Passthrough Float=0.6, Spawn Group Mask=1)
    Sprite Size Mode=::NewEnumerator3
    Uniform Sprite Size=&MaskFloatBySpawnGroup(Masked Float=40, Passthrough Float=20)
    UsePositionOffset=false
    Write Material Random=true
   )
   &CameraOffset(Camera Offset Amount=&MaskFloatBySpawnGroup(Masked Float=-20, Passthrough Float=10))
   &SkeletalMeshLocation(
    Bone / Socket Index=1
    Bone Sampling Mode=::NewEnumerator4
    MaskSpawnGroup=true
    Position Sampling=::NewEnumerator2
    Skeletal Mesh=$System.Skeletal Mesh
    Spawn Group Mask=1
   ) {module=SkeletalMeshLocation001}
  ]
  particleUpdateScript [
   &SkeletalMeshLocation(
    Bone / Socket Index=&MaskIntBySpawnGroup(Masked Int=0, Passthrough Int=1, Spawn Group Mask=0)
    Bone Sampling Mode=::NewEnumerator4
    MaskSpawnGroup=true
    Position Sampling=::NewEnumerator2
    Skeletal Mesh=$System.Skeletal Mesh
   )
   /SetVariables(Particles.Position=/Niagara/DynamicInputs/Helpers/MaskPosition(
     Mask Source:NiagaraInt32=0
     Mask Parameter:NiagaraInt32=?
     Mask:NiagaraInt32=0
     Masked Position:NiagaraPosition=$Particles.SkeletalMeshLocation.SampledPosition
     Passthrough Position:NiagaraPosition=$Particles.SkeletalMeshLocation001.SampledPosition
    )) {isSetParameters=true}
   &SolveForcesAndVelocity(Clamp Velocity=false, Limit Acceleration=false)
   &ParticleState()
   &SpriteRotationRate(Rotation Rate=&FloatFromCurve(
     FloatCurve=DataInterface() {propertyValues {Curve {
        defaultValue=0
        keys [
         {arriveTangent=4.1785, interpMode=RCIM_Cubic, leaveTangent=4.1785, tangentMode=RCTM_User, time=0.2036, value=0}
         {arriveTangent=1.1179, interpMode=RCIM_Cubic, leaveTangent=1.1179, tangentMode=RCTM_User, time=0.4, value=0.725}
         {arriveTangent=0.0224, interpMode=RCIM_Cubic, leaveTangent=0.0224, tangentMode=RCTM_User, time=1, value=1}
        ]
       }}}
     Scale Curve=400
    ))
   &ScaleSpriteSize(Uniform Curve Sprite Scale=DataInterface() {propertyValues {Curve {
       defaultValue=0
       keys [
        {arriveTangent=2.1946, interpMode=RCIM_Cubic, leaveTangent=2.1946, tangentMode=RCTM_User, time=0, value=0.3921}
        {arriveTangent=1.4716, interpMode=RCIM_Cubic, leaveTangent=1.4716, time=0.3973, value=1.3}
        {arriveTangent=-3.1752, interpMode=RCIM_Cubic, leaveTangent=-3.1752, time=0.6377, value=1.3306}
        {arriveTangent=0.481, interpMode=RCIM_Cubic, leaveTangent=0.481, tangentMode=RCTM_User, time=0.7469, value=0.1898}
        {arriveTangent=-0.7498, interpMode=RCIM_Cubic, leaveTangent=-0.7498, time=0.8486, value=2.9692}
        {arriveTangent=-4.3961, interpMode=RCIM_Cubic, leaveTangent=-4.3961, tangentMode=RCTM_User, time=1, value=0}
       ]
      }}})
   &DynamicMaterialParameters(
    Index 0 Param 1=1
    Index 0 Param 3=&FloatFromCurve(
     FloatCurve=DataInterface() {propertyValues {Curve {keys [{interpMode=RCIM_Cubic, leaveTangent=8.1085, tangentMode=RCTM_Break, time=0.8568, value=1}, {tangentMode=RCTM_User, time=1, value=0}]}}}
     Scale Curve=0.015
    )
    Index 0 Param 4=0.8
    Index 1 Param 1=&FloatFromCurve(FloatCurve=DataInterface() {propertyValues {Curve {keys [
         {interpMode=RCIM_Cubic, leaveTangent=1.1034, tangentMode=RCTM_Break, time=0, value=0}
         {arriveTangent=2.2399, tangentMode=RCTM_User, time=0.5, value=1}
        ]}}})
    Index 1 Param 2=&FloatFromCurve(FloatCurve=DataInterface() {propertyValues {Curve {keys [<at 0.5 0>, {arriveTangent=3.2053, tangentMode=RCTM_User, time=0.9, value=1}]}}})
    Index 1 Param 3=&FloatFromCurve(FloatCurve=DataInterface() {propertyValues {Curve {keys [<at 0 0 RCIM_Cubic>, {arriveTangent=3.4223, interpMode=RCIM_Cubic, leaveTangent=3.4223, tangentMode=RCTM_User, time=1, value=1}]}}})
    Index 1 Param 4=&FloatFromCurve(
     FloatCurve=DataInterface() {propertyValues {Curve {keys [
         {interpMode=RCIM_Cubic, leaveTangent=0.1169, tangentMode=RCTM_Break, time=0, value=0}
         {arriveTangent=2.5708, interpMode=RCIM_Cubic, leaveTangent=2.5708, tangentMode=RCTM_User, time=1, value=1}
        ]}}}
     Scale Curve=&RandomRangeFloat(Minimum=1, Maximum=40)
    )
    Index 2 Param 1=&FloatFromCurve(FloatCurve=DataInterface() {propertyValues {Curve {keys [
         {interpMode=RCIM_Cubic, leaveTangent=-0.0025, tangentMode=RCTM_Break, time=0, value=0.5}
         {arriveTangent=-0.0316, interpMode=RCIM_Cubic, tangentMode=RCTM_Break, time=1, value=-0.2}
        ]}}})
    Index 2 Param 2=&FloatFromCurve(
     FloatCurve=DataInterface() {propertyValues {Curve {keys [
         {interpMode=RCIM_Cubic, tangentMode=RCTM_User, time=0, value=0}
         {arriveTangent=2.0522, interpMode=RCIM_Cubic, leaveTangent=2.0522, tangentMode=RCTM_User, time=1, value=1}
        ]}}}
     Scale Curve=2
    )
    Module.Index0Float0Write=true
    Module.Index0Float1Write=true
    Module.Index0Float2Write=true
    Module.Index0Float3Write=true
    Module.Index1Float0Write=true
    Module.Index1Float1Write=true
    Module.Index1Float2Write=true
    Module.Index1Float3Write=true
    Module.Index2Float0Write=true
    Module.Index2Float1Write=true
    Module.Index2Float2Write=false
    Module.Index2Float3Write=false
    Module.Index3Float0Write=false
    Module.Index3Float1Write=false
    Module.Index3Float2Write=false
    Module.Index3Float3Write=true
    Write Parameter Index 1=::NewEnumerator1
    Write Parameter Index 2=::NewEnumerator1
    Write Parameter Index 3=::NewEnumerator1
   )
   &ScaleColor(
    Scale Alpha=&FloatFromCurve(
     FloatCurve=DataInterface() {propertyValues {Curve {keys [
         {tangentMode=RCTM_User, time=0, value=1}
         {arriveTangent=-0.0405, interpMode=RCIM_Cubic, leaveTangent=-0.0405, tangentMode=RCTM_User, time=0.7009, value=1}
         {arriveTangent=-8.0089, interpMode=RCIM_Cubic, leaveTangent=-8.0089, tangentMode=RCTM_User, time=1, value=0}
        ]}}}
     Scale Curve=&Multiply_Float(
      A=&FN_FrontEndFloatModifier(Gameplay=1, Front End=0.5)
      B=&MaskFloatBySpawnGroup(
       Masked Float=1
       Passthrough Float=&FloatFromCurve(
        FloatCurve=DataInterface() {propertyValues {Curve {
           defaultValue=1
           keys [
            {arriveTangent=8.9136, interpMode=RCIM_Cubic, leaveTangent=8.9136, tangentMode=RCTM_User, time=0, value=0}
            {
             arriveTangentWeight=0.1126
             interpMode=RCIM_Cubic
             leaveTangentWeight=0.1126
             tangentMode=RCTM_User
             tangentWeightMode=RCTWM_WeightedBoth
             time=0.3264
             value=1
            }
            {arriveTangent=-4.0894, interpMode=RCIM_Cubic, leaveTangent=-4.0894, tangentMode=RCTM_User, time=1.0009, value=0}
           ]
          }}}
        Scale Curve=&FNTimeOfDayScalar(Night Scale=0.7, Morning Scale=1, Evening Scale=1)
       )
      )
     )
    )
    Scale RGB=&Lerp_Vector(
     A=&VectorFromFloat(Value=&FloatFromCurve(
       FloatCurve=DataInterface() {propertyValues {Curve {keys [<at 0 1>,<at 0.1 1>,<at 0.2 2>,<at 0.3 1>,<at 0.6 1>,<at 0.7 2>,<at 0.8 1>,<at 1 1>]}}}
       Scale Curve=10
      ))
     B=&VectorFromFloat(Value=&RandomRangeFloat(Minimum=0.3, Maximum=5, Evaluation Type=::NewEnumerator1))
     Alpha=&FloatFromCurve(FloatCurve=DataInterface() {propertyValues {Curve {keys [
          {interpMode=RCIM_Cubic, leaveTangent=2.5294, tangentMode=RCTM_Break, time=0, value=0}
          {arriveTangent=-0.2166, interpMode=RCIM_Cubic, tangentMode=RCTM_Break, time=0.4, value=1}
         ]}}})
    )
    ScaleA=true
    ScaleRGB=false
   )
  ]
  rendererClasses [&NiagaraSpriteRendererProperties]
  renderers [{rendererClass=&NiagaraSpriteRendererProperties}]
 }
 Light {
  emitterUpdateScript [
   &EmitterState(Life Cycle Mode=::NewEnumerator1, Loop Behavior=::NewEnumerator1, Loop Delay=0.63, Loop Duration=5.1, UseLoopDelay=true)
   &SpawnBurst_Instantaneous(Spawn Group=1, Use Spawn Probability=false, UseLoopCountLimit=false)
  ]
  particleSpawnScript [
   &InitializeParticle(
    Color=$User.LightningColor
    Lifetime=0.25
    Position Offset=<0 0 -10>
    Sprite Size Mode=::NewEnumerator3
    Uniform Sprite Size=250
    UsePositionOffset=false
    Write Material Random=true
   )
   &CameraOffset(Camera Offset Amount=20)
   &SkeletalMeshLocation(Bone / Socket Index=1, Bone Sampling Mode=::NewEnumerator4, MaskSpawnGroup=false, Skeletal Mesh=$System.Skeletal Mesh)
  ]
  particleUpdateScript [
   &ParticleState()
   &SpriteRotationRate(Rotation Rate=&FloatFromCurve(
     FloatCurve=DataInterface() {propertyValues {Curve {keys [
         {interpMode=RCIM_Cubic, tangentMode=RCTM_User, time=0, value=1}
         {arriveTangent=-1.2549, interpMode=RCIM_Cubic, leaveTangent=-1.2549, tangentMode=RCTM_User, time=1, value=0}
        ]}}}
     Scale Curve=20
    ))
   &ScaleColor(
    Scale Alpha=&FloatFromCurve(FloatCurve=DataInterface() {propertyValues {Curve {
        defaultValue=1
        keys [
         {interpMode=RCIM_Cubic, tangentMode=RCTM_User, time=0, value=1}
         {arriveTangent=0.1313, interpMode=RCIM_Cubic, leaveTangent=0.1313, time=0.0904, value=1.3675}
         {arriveTangent=-0.0904, interpMode=RCIM_Cubic, leaveTangent=-0.0904, tangentMode=RCTM_User, time=0.1206, value=1.0158}
         {arriveTangent=0.4479, interpMode=RCIM_Cubic, leaveTangent=0.4479, time=0.8158, value=1.0409}
         {arriveTangent=-5.6527, interpMode=RCIM_Cubic, leaveTangent=-5.6527, time=0.872, value=1.3524}
         {arriveTangent=-2.9571, interpMode=RCIM_Cubic, leaveTangent=-2.9571, tangentMode=RCTM_User, time=1, value=0}
        ]
       }}})
    ScaleA=true
    ScaleRGB=false
   )
   /SetVariables(
    Particles.FlareDynPar:Vector4f=<1 0 0 0>
    Particles.FlareSpriteSize:Vector2f=&Vector2DFromFloat(Value=&FloatFromCurve(
      FloatCurve=DataInterface() {propertyValues {Curve {keys [
          {interpMode=RCIM_Cubic, leaveTangent=5.6981, tangentMode=RCTM_Break, time=0, value=0.2}
          <at 0.315 1>
          {arriveTangent=-0.209, interpMode=RCIM_Cubic, leaveTangent=-0.209, time=4.9, value=1}
          {arriveTangent=-13.1443, interpMode=RCIM_Cubic, leaveTangent=-13.1443, tangentMode=RCTM_User, time=5.1, value=0}
         ]}}}
      CurveIndex=$Particles.Age
      Scale Curve=390
     ))
   ) {isSetParameters=true}
   &DynamicMaterialParameters(
    Index 0 Param 1=1
    Index 0 Param 2=1
    Index 0 Param 3=1
    Index 0 Param 4=2.3
    Index 2 Param 1=2.5
    Index 2 Param 3=1
    Index 2 Param 4=1
    Index 3 Param 1=30
    Module.Index0Float0Write=true
    Module.Index0Float1Write=true
    Module.Index0Float2Write=true
    Module.Index0Float3Write=true
    Module.Index1Float0Write=true
    Module.Index1Float1Write=true
    Module.Index1Float2Write=true
    Module.Index1Float3Write=true
    Module.Index2Float0Write=true
    Module.Index2Float1Write=true
    Module.Index2Float2Write=true
    Module.Index2Float3Write=true
    Module.Index3Float0Write=true
    Module.Index3Float1Write=true
    Module.Index3Float2Write=true
    Module.Index3Float3Write=true
    Write Parameter Index 1=::NewEnumerator1
    Write Parameter Index 2=::NewEnumerator1
    Write Parameter Index 3=::NewEnumerator1
   )
  ]
  rendererClasses [&NiagaraSpriteRendererProperties]
  renderers [{rendererClass=&NiagaraSpriteRendererProperties}]
 }
 Main_Lightning {
  emitterUpdateScript [
   &EmitterState(Life Cycle Mode=::NewEnumerator1, Loop Behavior=::NewEnumerator1, Loop Delay=0.6, Loop Duration=1, UseLoopDelay=true)
   &SpawnBurst_Instantaneous(Use Spawn Probability=false, UseLoopCountLimit=false)
  ]
  particleSpawnScript [
   &PartitionParticles(Number Of Partitions=10, Particles Per Partition=1, Partition Mode:ENiagara_PartitionMode_v2=::NewEnumerator1)
   /SetVariables(Particles.GlobalRandom:NiagaraFloat=&RandomRangeFloat(Minimum=0.4, Maximum=1.2)) {isSetParameters=true}
   &InitializeParticle(
    Color=&Multiply_LinearColorByFloat(Linear Color=$User.LightningColor, Float=5)
    Lifetime=&RandomRangeFloat(Minimum=0.2, Maximum=0.35)
    Mesh Scale Mode=::NewEnumerator3
    Mesh Uniform Scale=&MakeFloatFromVector(
     Vector=&MakeVector(
      X=&RandomRangeFloat(Minimum=0.6, Maximum=1.1)
      Y=1.2
      Z=&Multiply_Float(A=2, B=$Particles.GlobalRandom)
     )
     Channel=::NewEnumerator0
    )
    Position Mode=::NewEnumerator1
    UsePositionOffset=false
    Write Material Random=true
   )
   &InitialMeshOrientation(
    Mesh Axis To Orient=<0 0 1>
    Mesh Orientation Mode=::NewEnumerator2
    Orientation Vector=<0 0 -1>
    Rotation=&RandomRangeVector(Minimum=<-0.01 -0.01 0>, Maximum=<0.01 0.01 0>)
    Rotation Coordinate Space=::NewEnumerator2
    Use Rotation Vector=true
   )
   &CameraOffset(Camera Offset Amount=20)
  ]
  particleUpdateScript [
   &ParticleState()
   &SkeletalMeshLocation(Bone / Socket Index=1, Bone Sampling Mode=::NewEnumerator4, MaskSpawnGroup=false, Skeletal Mesh=$System.Skeletal Mesh)
   &SolveForcesAndVelocity(Clamp Velocity=false, Limit Acceleration=false)
   &DynamicMaterialParameters(
    Index 0 Param 1=&FloatFromCurve(
     FloatCurve=DataInterface() {propertyValues {Curve {
        defaultValue=0
        keys [
         {arriveTangent=9.0725, interpMode=RCIM_Cubic, leaveTangent=9.0725, tangentMode=RCTM_User, time=0, value=0}
         {arriveTangent=-2.5249, interpMode=RCIM_Cubic, leaveTangent=-2.5249, tangentMode=RCTM_User, time=0.1408, value=0.9847}
         {arriveTangent=-0.0405, interpMode=RCIM_Cubic, leaveTangent=-0.0405, tangentMode=RCTM_User, time=1, value=0.2}
        ]
       }}}
     Scale Curve=&RandomRangeFloat(Minimum=1, Maximum=0.5)
    )
    Index 0 Param 2=&FloatFromCurve(
     FloatCurve=DataInterface() {propertyValues {Curve {
        defaultValue=0
        keys [<at 0 0.2 RCIM_Cubic>, {arriveTangent=2.4937, interpMode=RCIM_Cubic, leaveTangent=2.4937, tangentMode=RCTM_User, time=1, value=1}]
       }}}
     Scale Curve=&Multiply_Float(A=30, B=$Particles.GlobalRandom)
    )
    Index 0 Param 3=&RandomRangeFloat(Minimum=0.25, Maximum=1)
    Index 0 Param 4=&FloatFromCurve(FloatCurve=DataInterface() {propertyValues {Curve {
        defaultValue=0
        keys [
         {arriveTangent=3.552, interpMode=RCIM_Cubic, leaveTangent=3.552, tangentMode=RCTM_User, time=0, value=0}
         {arriveTangent=2, interpMode=RCIM_Cubic, leaveTangent=2, time=0.2385, value=0.7635}
         {arriveTangent=0.0224, interpMode=RCIM_Cubic, leaveTangent=0.0224, tangentMode=RCTM_User, time=0.5, value=1}
        ]
       }}})
    Index 1 Param 1=&FloatFromCurve(FloatCurve=DataInterface() {propertyValues {Curve {
        defaultValue=1
        keys [
         {interpMode=RCIM_Cubic, tangentMode=RCTM_User, time=0, value=0}
         {arriveTangent=2.9607, interpMode=RCIM_Cubic, leaveTangent=2.9607, tangentMode=RCTM_User, time=0.7, value=1}
        ]
       }}})
    Index 1 Param 2=&Multiply_Float(A=1, B=$Particles.GlobalRandom)
    Index 1 Param 3=&Multiply_Float(A=0.5, B=$Particles.GlobalRandom)
    Index 1 Param 4=&RandomRangeFloat(Minimum=0.5, Maximum=1)
    Module.Index0Float0Write=true
    Module.Index0Float1Write=true
    Module.Index0Float2Write=true
    Module.Index0Float3Write=true
    Module.Index1Float0Write=true
    Module.Index1Float1Write=true
    Module.Index1Float2Write=true
    Module.Index1Float3Write=true
    Write Parameter Index 1=::NewEnumerator1
   )
  ]
  rendererClasses [&NiagaraMeshRendererProperties]
  renderers [{rendererClass=&NiagaraMeshRendererProperties, rendererIndex=0}]
 }
 ParticleDecal {
  emitterUpdateScript [
   &EmitterState(Life Cycle Mode=::NewEnumerator1, Loop Behavior=::NewEnumerator1, Loop Delay=0.57, Loop Duration=1, UseLoopDelay=true)
   &SpawnBurst_Instantaneous(Use Spawn Probability=false, UseLoopCountLimit=false)
  ]
  particleSpawnScript [
   &InitializeParticle(
    Color=&Multiply_LinearColorByFloat(Linear Color=$User.LightningColor, Float=5)
    Lifetime=0.4
    Mesh Scale Mode=::NewEnumerator3
    Mesh Uniform Scale=2
    Position Offset=<0 0 100>
    Sprite Size Mode=::NewEnumerator3
    Uniform Sprite Size=150
    UsePositionOffset=false
    Write Material Random=true
   )
   &CameraOffset(Camera Offset Amount=30)
   &DynamicMaterialParameters(
    Index 0 Param 1=1
    Index 0 Param 3=1
    Module.Index0Param0Write=true
    Module.Index0Param1Write=true
    Module.Index0Param2Write=true
    Module.Index0Param3Write=true
   )
  ]
  particleUpdateScript [
   &ParticleState()
   &SkeletalMeshLocation(Bone / Socket Index=1, Bone Sampling Mode=::NewEnumerator4, MaskSpawnGroup=false, Skeletal Mesh=$System.Skeletal Mesh)
   &ScaleColor(
    Scale Alpha=&FloatFromCurve(FloatCurve=DataInterface() {propertyValues {Curve {
        defaultValue=1
        keys [
         {interpMode=RCIM_Cubic, tangentMode=RCTM_User, time=0.1247, value=1}
         {arriveTangent=-0.2743, interpMode=RCIM_Cubic, leaveTangent=-0.2743, tangentMode=RCTM_User, time=0.5942, value=0.1526}
         {arriveTangent=-0.5414, interpMode=RCIM_Cubic, leaveTangent=-0.5414, tangentMode=RCTM_User, time=1, value=0}
        ]
       }}})
    ScaleA=true
    ScaleRGB=false
   )
  ]
  rendererClasses [&NiagaraSpriteRendererProperties]
  renderers [{rendererClass=&NiagaraSpriteRendererProperties}]
 }
 Sparks {
  emitterUpdateScript [
   &EmitterState(Life Cycle Mode=::NewEnumerator1, Loop Behavior=::NewEnumerator1, Loop Delay=0.67, Loop Duration=1, UseLoopDelay=true)
   &SpawnBurst_Instantaneous(Spawn Count=30, Use Spawn Probability=false, UseLoopCountLimit=false)
  ]
  particleSpawnScript [
   &InitializeParticle(
    Color=$User.LightningColor
    Lifetime Max=0.15
    Lifetime Min=0.5
    Lifetime Mode=::NewEnumerator1
    Sprite Size Mode=::NewEnumerator1
    Uniform Sprite Size Max=3
    UsePositionOffset=false
    Write Material Random=true
   )
   &SkeletalMeshLocation(Bone / Socket Index=1, Bone Sampling Mode=::NewEnumerator4, MaskSpawnGroup=false, Skeletal Mesh=$System.Skeletal Mesh)
   &ShapeLocation(MaskSpawnGroup=false, Override Seed=false, Sphere Radius=1, UseOwnerScale=true)
   &AddVelocity(
    Module.Use Falloff Distance=false
    Override Seed=false
    Velocity Mode=::NewEnumerator1
    Velocity Origin=$Particles.ShapeLocation.ShapePosition
    Velocity Speed=&RandomRangeFloat(Minimum=1000, Maximum=600)
   )
  ]
  particleUpdateScript [
   &ParticleState()
   &Drag(
    Drag=&FloatFromCurve(
     FloatCurve=DataInterface() {propertyValues {Curve {
        defaultValue=1
        keys [
         {interpMode=RCIM_Cubic, tangentMode=RCTM_User, time=0, value=0}
         {arriveTangent=6.788, interpMode=RCIM_Cubic, leaveTangent=6.788, time=0.116, value=0.2056}
         {arriveTangent=15.8462, interpMode=RCIM_Cubic, leaveTangent=15.8462, tangentMode=RCTM_User, time=0.1473, value=1}
        ]
       }}}
     Scale Curve=&RandomRangeFloat(Minimum=8, Maximum=12)
    )
    Use Linear Drag=true
    Use Rotational Drag=true
   )
   &SolveForcesAndVelocity(Clamp Velocity=false, Limit Acceleration=false)
   &ScaleSpriteSizeBySpeed(Max Scale Factor=<0.5 8>, Min Scale Factor=<0.3 2>)
   &ScaleColor(
    Scale Alpha=&FloatFromCurve(FloatCurve=DataInterface() {propertyValues {Curve {
        defaultValue=1
        keys [
         {tangentMode=RCTM_User, time=0, value=1}
         {arriveTangent=-0.0405, interpMode=RCIM_Cubic, leaveTangent=-0.0405, tangentMode=RCTM_User, time=0.7009, value=1}
         {arriveTangent=-8.0089, interpMode=RCIM_Cubic, leaveTangent=-8.0089, tangentMode=RCTM_User, time=1, value=0}
        ]
       }}})
    Scale RGB=&Multiply_VectorByFloat(
     Vector=<1 1 1>
     Float=&FloatFromCurve(FloatCurve=DataInterface() {propertyValues {Curve {keys [<at 0 5>,<at 1 0>]}}})
    )
    ScaleA=true
    ScaleRGB=true
   )
  ]
  rendererClasses [&NiagaraSpriteRendererProperties]
  renderers [{rendererClass=&NiagaraSpriteRendererProperties}]
 }
}
summary {
 systemName=NS_Emote_FumeFleeceClap_SpawnIn
 emitters [
  {emitterName=Flash, rendererClasses [&NiagaraSpriteRendererProperties]}
  {emitterName=Main_Lightning, rendererClasses [&NiagaraMeshRendererProperties]}
  {emitterName=ParticleDecal, rendererClasses [&NiagaraSpriteRendererProperties]}
  {emitterName=Sparks, rendererClasses [&NiagaraSpriteRendererProperties]}
  {emitterName=Light, rendererClasses [&NiagaraSpriteRendererProperties]}
 ]
 userVariables [
  User.bIsFrontEnd {type=NiagaraBool}
  User.bIsFrontEndPreview {type=NiagaraBool}
  User.LightningColor {type=LinearColor, description="", defaultValue {struct=LinearColor, value=<rgba 0 0.3269 3 1>}}
 ]
}
